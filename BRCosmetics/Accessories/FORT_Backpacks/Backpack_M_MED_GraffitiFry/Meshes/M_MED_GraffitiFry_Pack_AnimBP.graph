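# Animation Blueprint for the GraffitiFry backpack cosmetic, adding physics sway and state-based poses to its winged pack.
# Anim Dynamics springs drive both wing base bones from the reference pose, then cached poses add bone offsets for sprinting and skydive diving.
# Bool blends choose among these, then emote and skydiving flags toggle further additive root, pack and wing offsets.
# Each update sets GravityOverride through a deprecated helper, while the dynamics nodes carry fixed gravity values.
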AnimBlueprint /BRCosmetics/Accessories/FORT_Backpacks/Backpack_M_MED_GraffitiFry/Meshes/M_MED_GraffitiFry_Pack_AnimBP
target_skeleton=&/M_MED_GraffitiFry_Pack_Skeleton
graphs [
 AnimGraph {nodes [
   n1 = "Transform (Modify) Bone - Bone: root"!(
    ComponentPose=Local To Component(LocalPose=Use cached pose 'CRCache_BlendPoses'())
    bAlphaBoolEnabled:bool=Get bIsPlayingEmote()
    Translation="0,-1.5,0"
    Rotation="0,0,15"
    Scale
   ) -> (Pose) {
    alphaInputType=Bool
    rotationMode=BMM_Additive
    rotationSpace=BCS_BoneSpace
    translationMode=BMM_Additive
    translationSpace=BCS_BoneSpace
    alphaBoolBlend {blendOption=Linear}
    boneToModify {boneName=root}
    rotation {roll=15}
    translation {y=-1.5}
   }
   n2 = Get bIsSkyDiving()
   n3 = "Transform (Modify) Bone - Bone: pack"!(ComponentPose=@n1, bAlphaBoolEnabled:bool=@n2, Translation="0,-2,0", Rotation="0,0,20", Scale) -> (Pose) {
    alphaInputType=Bool
    rotationMode=BMM_Additive
    rotationSpace=BCS_BoneSpace
    translationMode=BMM_Additive
    translationSpace=BCS_BoneSpace
    alphaBoolBlend {blendOption=Linear}
    boneToModify {boneName=pack}
    rotation {roll=20}
    translation {y=-2}
   }
   n4 = "Transform (Modify) Bone - Bone: dyn_wing_a_l"!(ComponentPose=@n3, bAlphaBoolEnabled:bool=@n2, Translation="0,0,0", Rotation="20,-20,0", Scale="1,1,1") -> (Pose) {
    alphaInputType=Bool
    rotationMode=BMM_Additive
    rotationSpace=BCS_BoneSpace
    translationMode=BMM_Additive
    translationSpace=BCS_BoneSpace
    alphaBoolBlend {blendOption=Linear}
    boneToModify {boneName=dyn_wing_a_l}
    rotation {pitch=20, yaw=-20}
   }
   n5 = "Transform (Modify) Bone - Bone: dyn_wing_a_r"!(ComponentPose=@n4, bAlphaBoolEnabled:bool=@n2, Translation="0,0,0", Rotation="20,-20,0", Scale="1,1,1") -> (Pose) {
    alphaInputType=Bool
    rotationMode=BMM_Additive
    rotationSpace=BCS_BoneSpace
    translationMode=BMM_Additive
    translationSpace=BCS_BoneSpace
    alphaBoolBlend {blendOption=Linear}
    boneToModify {boneName=dyn_wing_a_r}
    rotation {pitch=20, yaw=-20}
   }
   Output Pose(Result=Component To Local(ComponentPose=@n5)) {layerGroup=None}
   n6 = Anim Dynamics - Bone: dyn_base_wing_a_l(ComponentPose=Local To Component(LocalPose=Local Space Ref Pose()), Alpha=1, ExternalForce="0,0,0") {
    alpha=0.5
    angularDampingOverride=0.9999
    angularSpringConstant=4000
    bAngularSpring=true
    bGravityOverrideInSimSpace=true
    bOverrideAngularDamping=true
    bUseGravityOverride=true
    linearDampingOverride=0.8
    alphaBoolBlend {blendOption=Linear}
    boundBone {boneName=dyn_base_wing_a_l}
    componentAppliedLinearAccClamp=<5000 5000 5000>
    componentLinearAccScale=<5 3 3>
    componentLinearVelScale=<3 3 3>
    gravityOverride {z=-980}
    physicsBodyDefinitions [{
      sphereCollisionRadius=0
      boundBone {boneName=dyn_base_wing_a_l}
      boxExtents=<10 20 20>
      constraintSetup {angularLimitsMax=<0 0 2>, angularLimitsMin=<0 -20 -5>, angularTarget=<1 0 0>}
      localJointOffset=<0.3 0 0>
     }]
    relativeSpaceBone {boneName=pack}
   }
   n7 = Anim Dynamics - Bone: dyn_base_wing_a_r(ComponentPose=@n6, Alpha=1, ExternalForce="0,0,0") {
    alpha=0.5
    angularDampingOverride=0.9999
    angularSpringConstant=4000
    bAngularSpring=true
    bGravityOverrideInSimSpace=true
    bOverrideAngularDamping=true
    bUseGravityOverride=true
    linearDampingOverride=0.8
    alphaBoolBlend {blendOption=Linear}
    boundBone {boneName=dyn_base_wing_a_r}
    componentAppliedLinearAccClamp=<5000 5000 5000>
    componentLinearAccScale=<5 3 3>
    componentLinearVelScale=<3 3 3>
    gravityOverride {z=-980}
    physicsBodyDefinitions [{
      sphereCollisionRadius=0
      boundBone {boneName=dyn_base_wing_a_r}
      boxExtents=<10 20 20>
      constraintSetup {angularLimitsMax=<0 0 2>, angularLimitsMin=<0 -20 -5>, angularTarget=<1 0 0>}
      localJointOffset=<-0.3 0 0>
     }]
    relativeSpaceBone {boneName=pack}
   }
   n8 = "Transform (Modify) Bone - Bone: root"!(ComponentPose=@n7, Alpha:float, Translation="0,-1.5,0", Rotation="0,0,10", Scale) -> (Pose) {
    rotationMode=BMM_Additive
    rotationSpace=BCS_BoneSpace
    translationMode=BMM_Additive
    translationSpace=BCS_BoneSpace
    alphaBoolBlend {blendOption=Linear}
    boneToModify {boneName=root}
    rotation {roll=10}
    translation {y=-1.5}
   }
   Save cached pose 'CRCache_Input'(Pose=Component To Local(ComponentPose=@n8)) {cacheName=CRCache_Input}
   n9 = "Transform (Modify) Bone - Bone: dyn_wing_a_l"(
    ComponentPose=Local To Component(LocalPose=Use cached pose 'CRCache_Input'())
    bAlphaBoolEnabled=True
    Translation="0,0,0"
    Rotation="-20,20,0"
    Scale="1,1,1"
   ) {
    alphaInputType=Bool
    rotationMode=BMM_Additive
    rotationSpace=BCS_BoneSpace
    translationMode=BMM_Additive
    translationSpace=BCS_BoneSpace
    alphaBoolBlend {blendOption=Linear}
    boneToModify {boneName=dyn_wing_a_l}
    rotation {pitch=-20, yaw=20}
   }
   n10 = "Transform (Modify) Bone - Bone: dyn_wing_a_r"(ComponentPose=@n9, bAlphaBoolEnabled=True, Translation="0,0,0", Rotation="-20,20,0", Scale="1,1,1") {
    alphaInputType=Bool
    rotationMode=BMM_Additive
    rotationSpace=BCS_BoneSpace
    translationMode=BMM_Additive
    translationSpace=BCS_BoneSpace
    alphaBoolBlend {blendOption=Linear}
    boneToModify {boneName=dyn_wing_a_r}
    rotation {pitch=-20, yaw=20}
   }
   Save cached pose 'CRCache_SkyDive_Input'(Pose=Component To Local(ComponentPose=@n10)) {cacheName=CRCache_SkyDive_Input}
   n11 = Use cached pose 'CRCache_SkyDive_Input'()
   n12 = Use cached pose 'CRCache_Input'()
   n13 = Get bIsSkydiveDiveMode()
   Save cached pose 'CRCache_BlendPoses'(Pose=Blend Poses by bool(
     BlendPose_0=Use cached pose 'CRCache_Sprinting_Input'()
     BlendPose_1=Blend Poses by bool(BlendPose_0=@n11, BlendPose_1=@n12, bActiveValue=@n13) {blendType=Linear}
     bActiveValue=Get bIsSprinting()
    ) {blendType=Linear}) {cacheName=CRCache_BlendPoses}
   n14 = "Transform (Modify) Bone - Bone: dyn_base_wing_a_r"!(
    ComponentPose=Local To Component(LocalPose=Use cached pose 'CRCache_Input'())
    bAlphaBoolEnabled:bool=True
    Translation="0,0,0"
    Rotation="0,0,10"
    Scale="1,1,1"
   ) -> (Pose) {
    alphaInputType=Bool
    rotationMode=BMM_Additive
    rotationSpace=BCS_BoneSpace
    alphaBoolBlend {blendOption=Cubic}
    boneToModify {boneName=dyn_base_wing_a_r}
    rotation {roll=10}
   }
   n15 = "Transform (Modify) Bone - Bone: dyn_base_wing_a_l"!(ComponentPose=@n14, bAlphaBoolEnabled:bool=True, Translation="0,0,0", Rotation="0,0,10", Scale="1,1,1") -> (Pose) {
    alphaInputType=Bool
    rotationMode=BMM_Additive
    rotationSpace=BCS_BoneSpace
    alphaBoolBlend {blendOption=Cubic}
    boneToModify {boneName=dyn_base_wing_a_l}
    rotation {roll=10}
   }
   Save cached pose 'CRCache_Sprinting_Input'(Pose=Component To Local(ComponentPose=@n15)) {cacheName=CRCache_Sprinting_Input}
  ]}
 EventGraph {nodes [
   Event BlueprintUpdateAnimation()
   > n1 = Gravity Override State Function DEPRECATED(CharacterPartAnimInstance=Self-Reference(), GameTimeInSeconds=GetGameTimeInSeconds(), GravityOverrideParams=Get Gravity Override Params())
   > Set GravityOverride(GravityOverride=@n1)
  ]}
]
variables [
 Gravity Override Params {type=GravityOverrideParamsStruct}
 GravityOverride {type=Vector}
]
