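# Physics asset for the GraffitiFry medium male backpack cosmetic mesh.
# It defines a single capsule body on the pack bone, offset and rotated to fit the bag's volume.
# It has no constraints, so the pack collides as one rigid shape with no jointed or simulated sub-parts.
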PhysicsAsset /BRCosmetics/Accessories/FORT_Backpacks/Backpack_M_MED_GraffitiFry/Meshes/M_MED_GraffitiFry_Pack_PhysicsAsset
bodies {pack [{shapeType=Capsule, length=25.218, radius=12.3372, center=<0.0437 -5.7289 -8.6633>, rotation=<rot 0 89.9542 -179.8298>}]}
constraints []
